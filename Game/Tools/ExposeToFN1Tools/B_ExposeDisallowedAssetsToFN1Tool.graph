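# Editor utility blueprint tool that runs ExposeDisallowedAssets from the Fort Expose Disallowed Assets Tool on the UEFN_Disallowed collection file.
# Its Run event builds the file path by appending the cooked WindowsClient build's Saved/Collections folder and the file name to the project directory's parent.
# The RemoveProjectName helper derives that parent by searching backward for a slash and trimming the project folder name.
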EditorUtilityBlueprint /Game/Tools/ExposeToFN1Tools/B_ExposeDisallowedAssetsToFN1Tool
graphs [
 EventGraph {nodes [
   Event Run()
   n1 = ProjectDir!() -> (ReturnValue:String)
   > ExposeDisallowedAssets!(
    execute
    self:Fort Expose Disallowed Assets Tool Object Reference
    FilePath:String=Append!(
     A=RemoveProjectName!(ProjectDir:String=@n1) -> (ProjectDirWithoutProjectName:String)
     B=MakeLiteralString(Value=CookedBuild/WindowsClient/FortniteGame/Saved/Collections/)
     C:String=MakeLiteralString(Value=UEFN_Disallowed.collection)
    ) -> (ReturnValue)
   ) -> (then)
  ]}
 RemoveProjectName {nodes [
   n1 = RemoveProjectName!() -> (then, ProjectDir:String) {is=title K2Node_FunctionEntry}
   n2 = FindSubstring(SearchIn=@n1, Substring=/, bSearchFromEnd=true, StartPosition=int - int(A=Len(S=@n1)))
   > Return Node!(
    execute
    ProjectDirWithoutProjectName:String=GetSubstring!(SourceString:String=@n1, StartIndex:int=0, Length:int=int + int(A=@n2)) -> (ReturnValue:String)
   )
  ]}
]
variables []
